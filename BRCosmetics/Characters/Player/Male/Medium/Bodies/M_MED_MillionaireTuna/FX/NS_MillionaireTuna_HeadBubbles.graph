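# Niagara cosmetic effect that adds orange bubble sprites around the head of the Millionaire Tuna character.
# Two near-identical bubble emitters differ in size and brightness, spawning on a cylinder ring at the FaceAcc mesh's head bone.
# Acceleration, drag, attraction and two sphere blocking volumes shape motion; a region mask drives sprite shrinking, and sunlight color tints them.
# An "Empty" emitter bursts a large persistent sprite whose brightness follows time of day.
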NiagaraSystem /BRCosmetics/Characters/Player/Male/Medium/Bodies/M_MED_MillionaireTuna/FX/NS_MillionaireTuna_HeadBubbles
dependencies {
 usedDataInterfaces [&NiagaraDataInterfaceSkeletalMesh, &NiagaraDataInterfaceDebugDraw, &NiagaraDataInterfaceCurve]
 usedDynamicInputs [&Lerp_Float]
 usedModules [&SystemState]
}
emitters {
 Bubbles {
  emitterUpdateScript [
   &EmitterState()
   &SpawnRate(SpawnRate=8, Use Spawn Probability=false)
  ]
  particleSpawnScript [
   &InitializeParticle(
    Color=&MakeLinearColorFromVectorAndFloat("Vector (RGB)"=&MakeVectorFromLinearColorRGB(Color=<rgba 1 0.3538 0 1>), "Float (Alpha)"=1)
    Lifetime Min=1.75
    Lifetime Mode=::NewEnumerator1
    Sprite Size Mode=::NewEnumerator1
    Uniform Sprite Size Max=2
    Uniform Sprite Size Min=1
    UsePositionOffset=false
    Write Material Random=true
   )
   &SkeletalMeshLocation(
    Bone Sampling Mode=::NewEnumerator4
    MaskSpawnGroup=false
    Skeletal Mesh=DataInterface() {propertyValues {
      DefaultMesh=/BRCosmetics/Characters/Player/Male/Medium/Bodies/M_MED_MillionaireTuna/Meshes/Parts/M_MED_MillionaireTuna_FaceAcc
      ExcludeBoneName=root
      OptionalComponentClass=null
      PreviewMesh=&M_MED_MillionaireTuna_FaceAcc
      SkinningMode=SkinOnTheFly
      SoftSourceActor=null
      SourceMode=Default
      UvSetIndex=0
      WholeMeshLOD=-1
      bExcludeBone=false
      bReadDeformedGeometry=true
      bRequireCurrentFrameData=true
      ComponentTags []
      FilteredBones [head]
      FilteredSockets []
      MeshUserParameter {parameter {name=None, typeDefHandle {registeredTypeIndex=1}, varData []}}
      SamplingRegions []
     }}
   )
   &CylinderLocation(
    Cylinder Height=0
    Cylinder Height Midpoint=0
    Cylinder Origin:NiagaraPosition=/Niagara/DynamicInputs/Helpers/SimulationPosition()
    Cylinder Radius=8
    MaskSpawnGroup=false
    Offset=<-8 2 0>
    Offset Coordinate Space:ENiagaraCoordinateSpace=::Local
    Offset Mode:ENiagara_OffsetMode=::NewEnumerator0
    Override Seed=false
    Rotation Coordinate Space:ENiagaraCoordinateSpace=::Simulation
    Rotation Mode:ENiagara_RotationMode=::NewEnumerator1
    Scale Mode:ENiagara_ScaleMode=::NewEnumerator1
    Surface Only=true
    Use Endcaps In Surface Only Mode=false
   )
   &ShapeLocation(MaskSpawnGroup=false, Override Seed=false, UseOwnerScale=true) {enabled=false}
   &CameraOffset()
  ]
  particleUpdateScript [
   &ParticleState()
   &AccelerationForce(Acceleration=<15 0 0>, Coordinate Space=::Local)
   &Drag(Drag=&RandomRangeFloat(Minimum=0.5, Maximum=2), Use Linear Drag=true, Use Rotational Drag=false)
   &PointAttractionForce(
    Attraction Radius=20
    Attraction Strength=5
    Attractor Position Offset:Vector3f=<15 0 0>
    Attractor Position Offset Coordinate Space:ENiagaraCoordinateSpace=::Local
    Falloff Exponent=0.25
    Kill Within Radius=false
    Module.Use Falloff=true
    UseAttractorOffset=true
   )
   &SolveForcesAndVelocity(Clamp Velocity=false, Limit Acceleration=false)
   &self:SphereBlockingVolume(
    Pivot Offset:Vector3f=<8 0 0>
    Sphere Origin:Vector3f=&SimulationPosition()
    Sphere Radius:NiagaraFloat=&RandomRangeFloat(Minimum=9, Maximum=11)
   ) {module=SphereBlockingVolume}
   &self:SphereBlockingVolume(
    Pivot Offset:Vector3f=<8 0 0>
    Sphere Origin:Vector3f=&SimulationPosition()
    Sphere Radius:NiagaraFloat=&RandomRangeFloat(Minimum=7, Maximum=9)
   ) {module=SphereBlockingVolume001}
   &CalculateAccurateVelocity()
   &RegionMask(
    Invert Region=true
    Offset:Vector3f=<8 2 0>
    Offset Space:ENiagaraCoordinateSpace=::Local
    Region Radius=16
    UseOffset=true
    UseRotation=false
   )
   &DynamicMaterialParameters(
    Index 0 Param 1=&FloatFromCurve(FloatCurve=DataInterface() {propertyValues {Curve {keys [<at 0.7 1>,<at 1 0>]}}}, Scale Curve=2)
    Module.Param0WriteEnabled=true
    Module.Param1WriteEnabled=true
    Module.Param2WriteEnabled=true
    Module.Param3WriteEnabled=true
   )
   &ScaleSpriteSize(
    Uniform Curve Index=$Output.RegionMask.RegionMask
    Uniform Curve Sprite Scale=DataInterface() {propertyValues {Curve {keys [{tangentMode=RCTM_User, time=0.75, value=1}, {tangentMode=RCTM_User, time=1, value=0}]}}}
   ) {module=ScaleSpriteSize001}
   &ScaleColor(
    Scale RGB=&MakeVectorFromLinearColorRGB(Color=$NPC.FortniteNPC.SunLightColor)
    ScaleA=true
    ScaleRGB=true
   )
  ]
  rendererClasses [&NiagaraSpriteRendererProperties]
  renderers [{rendererClass=&NiagaraSpriteRendererProperties}]
 }
 Bubbles001 {
  emitterUpdateScript [
   &EmitterState()
   &SpawnRate(Use Spawn Probability=false)
  ]
  particleSpawnScript [
   &InitializeParticle(
    Color=&MakeLinearColorFromVectorAndFloat("Vector (RGB)"=&MakeVectorFromLinearColorRGB(Color=<rgba 10 3.5379 0 0.75>), "Float (Alpha)"=1)
    Lifetime Min=1.5
    Lifetime Mode=::NewEnumerator1
    Sprite Size Mode=::NewEnumerator1
    Uniform Sprite Size Max=4
    Uniform Sprite Size Min=2
    UsePositionOffset=false
    Write Material Random=true
   )
   &SkeletalMeshLocation(
    Bone Sampling Mode=::NewEnumerator4
    MaskSpawnGroup=false
    Skeletal Mesh=DataInterface() {propertyValues {
      DefaultMesh=&M_MED_MillionaireTuna_FaceAcc
      ExcludeBoneName=root
      OptionalComponentClass=null
      PreviewMesh=&M_MED_MillionaireTuna_FaceAcc
      SkinningMode=SkinOnTheFly
      SoftSourceActor=null
      SourceMode=Default
      UvSetIndex=0
      WholeMeshLOD=-1
      bExcludeBone=false
      bReadDeformedGeometry=true
      bRequireCurrentFrameData=true
      ComponentTags []
      FilteredBones [head]
      FilteredSockets []
      MeshUserParameter {parameter {name=None, typeDefHandle {registeredTypeIndex=1}, varData []}}
      SamplingRegions []
     }}
   )
   &CylinderLocation(
    Cylinder Height=0
    Cylinder Height Midpoint=0
    Cylinder Origin:NiagaraPosition=&SimulationPosition()
    Cylinder Radius=8
    MaskSpawnGroup=false
    Offset=<-8 2 0>
    Offset Coordinate Space:ENiagaraCoordinateSpace=::Local
    Offset Mode:ENiagara_OffsetMode=::NewEnumerator0
    Override Seed=false
    Rotation Coordinate Space:ENiagaraCoordinateSpace=::Simulation
    Rotation Mode:ENiagara_RotationMode=::NewEnumerator1
    Scale Mode:ENiagara_ScaleMode=::NewEnumerator1
    Surface Only=true
    Use Endcaps In Surface Only Mode=false
   )
   &ShapeLocation(MaskSpawnGroup=false, Override Seed=false, UseOwnerScale=true) {enabled=false}
   &CameraOffset()
  ]
  particleUpdateScript [
   &ParticleState()
   &AccelerationForce(Acceleration=<15 0 0>, Coordinate Space=::Local)
   &Drag(Drag=&RandomRangeFloat(Minimum=0.5, Maximum=2), Use Linear Drag=true, Use Rotational Drag=false)
   &PointAttractionForce(
    Attraction Radius=20
    Attraction Strength=5
    Attractor Position Offset:Vector3f=<15 0 0>
    Attractor Position Offset Coordinate Space:ENiagaraCoordinateSpace=::Local
    Falloff Exponent=0.25
    Kill Within Radius=false
    Module.Use Falloff=true
    UseAttractorOffset=true
   )
   &SolveForcesAndVelocity(Clamp Velocity=false, Limit Acceleration=false)
   &self:SphereBlockingVolume(
    Pivot Offset:Vector3f=<8 0 0>
    Sphere Origin:Vector3f=&SimulationPosition()
    Sphere Radius:NiagaraFloat=&RandomRangeFloat(Minimum=9, Maximum=11)
   ) {module=SphereBlockingVolume}
   &self:SphereBlockingVolume(
    Pivot Offset:Vector3f=<8 0 0>
    Sphere Origin:Vector3f=&SimulationPosition()
    Sphere Radius:NiagaraFloat=&RandomRangeFloat(Minimum=7, Maximum=9)
   ) {module=SphereBlockingVolume001}
   &CalculateAccurateVelocity()
   &RegionMask(
    Invert Region=true
    Offset:Vector3f=<8 2 0>
    Offset Space:ENiagaraCoordinateSpace=::Local
    Region Radius=16
    UseOffset=true
    UseRotation=false
   )
   &DynamicMaterialParameters(
    Index 0 Param 1=&FloatFromCurve(FloatCurve=DataInterface() {propertyValues {Curve {keys [<at 0.7 1>,<at 1 0>]}}}, Scale Curve=2)
    Index 0 Param 2=1
    Index 0 Param 3=1
    Module.Param0WriteEnabled=true
    Module.Param1WriteEnabled=true
    Module.Param2WriteEnabled=true
    Module.Param3WriteEnabled=true
   )
   &ScaleSpriteSize(
    Uniform Curve Index=$Output.RegionMask.RegionMask
    Uniform Curve Sprite Scale=DataInterface() {propertyValues {Curve {keys [{tangentMode=RCTM_User, time=0.75, value=1}, {tangentMode=RCTM_User, time=1, value=0}]}}}
   ) {module=ScaleSpriteSize001}
   &ScaleColor(
    Scale RGB=&MakeVectorFromLinearColorRGB(Color=$NPC.FortniteNPC.SunLightColor)
    ScaleA=true
    ScaleRGB=true
   )
  ]
  rendererClasses [&NiagaraSpriteRendererProperties]
  renderers [{rendererClass=&NiagaraSpriteRendererProperties}]
 }
 Empty {
  emitterUpdateScript [
   &EmitterState(Life Cycle Mode=::NewEnumerator1, Loop Behavior=::NewEnumerator1, Loop Duration=1, UseLoopDelay=false)
   &SpawnBurst_Instantaneous(Use Spawn Probability=false, UseLoopCountLimit=false)
  ]
  particleSpawnScript [
   &InitializeParticle(
    Color=<rgba 0.2344 0.0859 0 3>
    Lifetime=5
    Mesh Scale Mode=::NewEnumerator3
    Mesh Uniform Scale=5
    Position Offset=<6 4 0>
    Sprite Size Mode=::NewEnumerator3
    Uniform Sprite Size=30
    UsePositionOffset=true
    Write Material Random=true
   )
   &SampleSkeletalMesh(
    AdditionalYawPitchRoll=false
    Bone Sampling Mode:ENiagaraBoneSamplingMode=::NewEnumerator4
    Position Offset Coordinate Space:ENiagaraCoordinateSpace=::Local
    Skeletal Mesh=DataInterface() {propertyValues {
      DefaultMesh=&M_MED_MillionaireTuna_FaceAcc
      ExcludeBoneName=root
      OptionalComponentClass=null
      PreviewMesh=&M_MED_MillionaireTuna_FaceAcc
      SkinningMode=SkinOnTheFly
      SoftSourceActor=null
      SourceMode=Default
      UvSetIndex=0
      WholeMeshLOD=-1
      bExcludeBone=false
      bReadDeformedGeometry=true
      bRequireCurrentFrameData=true
      ComponentTags []
      FilteredBones []
      FilteredSockets []
      MeshUserParameter {parameter {name=None, typeDefHandle {registeredTypeIndex=1}, varData []}}
      SamplingRegions []
     }}
   )
   &CameraOffset(Camera Offset Amount=4)
  ]
  particleUpdateScript [
   &ParticleState(Kill Particles When Lifetime Has Elapsed=false)
   &DynamicMaterialParameters(
    Index 0 Param 1=2
    Index 0 Param 2=8
    Module.Index0Param0Write=true
    Module.Index0Param1Write=true
    Module.Index0Param2Write=true
    Module.Index0Param3Write=true
   )
   &ScaleColor(
    Scale RGB=&Multiply_VectorByFloat(
     Vector=<1 1 1>
     Float=&FN_Select_Float_from_Time_Of_Day(Morning=0.5, Day=1, Evening=0.5, Night=0.5)
    )
    ScaleA=true
    ScaleRGB=true
   )
  ]
  rendererClasses [&NiagaraSpriteRendererProperties]
  renderers [{rendererClass=&NiagaraSpriteRendererProperties}]
 }
}
summary {
 systemName=NS_MillionaireTuna_HeadBubbles
 emitters [
  {emitterName=Bubbles, rendererClasses [&NiagaraSpriteRendererProperties]}
  {emitterName=Empty, rendererClasses [&NiagaraSpriteRendererProperties]}
  {emitterName=Bubbles001, rendererClasses [&NiagaraSpriteRendererProperties]}
 ]
 userVariables []
}
