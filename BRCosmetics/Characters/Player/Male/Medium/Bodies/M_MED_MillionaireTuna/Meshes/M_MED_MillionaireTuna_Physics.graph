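# Physics asset for the M_MED_MillionaireTuna male medium player body cosmetic, setting up secondary motion for hanging chains and flaps around the hips.
# Capsule dyn_ bodies hinge off base_ anchor bodies with per-joint swing and twist limits, with two-link front and rear chains, a single center chain and four flaps.
# The anchors and both thigh capsules join the pelvis through angularly free constraints.
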PhysicsAsset /BRCosmetics/Characters/Player/Male/Medium/Bodies/M_MED_MillionaireTuna/Meshes/M_MED_MillionaireTuna_Physics
bodies {
 base_chain_c [{shapeType=Capsule, length=0.505, radius=0.505, shapeName=base_chain_c_capsule, rotation=<rot 0 -90 0>}]
 base_chain_front [{shapeType=Capsule, length=1, radius=1, shapeName=base_chain_front_capsule, rotation=<rot -90 0 0>}]
 base_chain_rear [{shapeType=Capsule, length=1, radius=1, shapeName=base_chain_rear_capsule, rotation=<rot -90 0 0>}]
 base_flap_l_front [{shapeType=Capsule, length=1, radius=1, shapeName=base_flap_l_front_capsule, rotation=<rot -90 0 0>}]
 base_flap_l_rear [{shapeType=Capsule, length=1, radius=1, shapeName=base_flap_l_rear_capsule, rotation=<rot -90 0 0>}]
 base_flap_r_front [{shapeType=Capsule, length=1, radius=1, shapeName=base_flap_r_front_capsule, rotation=<rot -90 0 0>}]
 base_flap_r_rear [{shapeType=Capsule, length=1, radius=1, shapeName=base_flap_r_rear_capsule, rotation=<rot -90 0 0>}]
 dyn_chain_1_front [{shapeType=Capsule, length=2, radius=1, shapeName=dyn_chain_1_front_capsule, center=<1 0 0>, rotation=<rot -90 0 0>}]
 dyn_chain_1_rear [{shapeType=Capsule, length=2, radius=1.5, shapeName=dyn_chain_1_rear_capsule, center=<1 0 -0.161>, rotation=<rot -90 -174.7087 174.7088>}]
 dyn_chain_2_front [{shapeType=Capsule, length=3, radius=1.5, shapeName=dyn_chain_2_front_capsule, center=<3 0 0>, rotation=<rot -90 25.3725 -25.3725>}]
 dyn_chain_2_rear [{shapeType=Capsule, length=2, radius=1.5, shapeName=dyn_chain_2_rear_capsule, center=<1 0 0>, rotation=<rot -90 0 0>}]
 dyn_chain_c [{shapeType=Capsule, length=8, radius=1, shapeName=dyn_chain_c_capsule, center=<8 0 5>, rotation=<rot -90 0 0>}]
 dyn_flap_l_front [{shapeType=Capsule, length=2, radius=1.5, shapeName=dyn_flap_l_front_capsule, center=<4 0 0>, rotation=<rot -90 0 0>}]
 dyn_flap_l_rear [{shapeType=Capsule, length=2, radius=2, shapeName=dyn_flap_l_rear_capsule, center=<4 0 0>, rotation=<rot -90 180 180>}]
 dyn_flap_r_front [{shapeType=Capsule, length=2, radius=1.5, shapeName=dyn_flap_r_rear_capsule, center=<4 0 0>, rotation=<rot -90 0 0>}]
 dyn_flap_r_rear [{shapeType=Capsule, length=2, radius=2, shapeName=dyn_flap_r_rear_capsule, center=<4 0 0>, rotation=<rot -90 0 0>}]
 pelvis [{shapeType=Capsule, length=9, radius=8.6}]
 thigh_l [{shapeType=Capsule, length=38.5, radius=8.25, center=<-18 1.6806 -0.3232>, rotation=<rot 0 90 -90>}]
 thigh_r [{shapeType=Capsule, length=38.5, radius=7.5, center=<18 -0.9974 0>, rotation=<rot 0 90 -90>}]
}
constraints [
 {bone1Name=dyn_chain_1_front, bone2Name=base_chain_front, swing1LimitDegrees=25, swing2LimitDegrees=35, twistLimitDegrees=5}
 {bone1Name=dyn_chain_2_front, bone2Name=dyn_chain_1_front, swing1LimitDegrees=25, swing2LimitDegrees=35, twistLimitDegrees=5}
 {bone1Name=dyn_chain_1_rear, bone2Name=base_chain_rear, swing1LimitDegrees=30, swing2LimitDegrees=30, twistLimitDegrees=0}
 {bone1Name=dyn_chain_2_rear, bone2Name=dyn_chain_1_rear, swing1LimitDegrees=35, swing2LimitDegrees=45, twistLimitDegrees=10}
 {bone1Name=dyn_flap_l_front, bone2Name=base_flap_l_front, swing1LimitDegrees=15, swing2LimitDegrees=25, twistLimitDegrees=0}
 {bone1Name=dyn_flap_l_rear, bone2Name=base_flap_l_rear, swing1LimitDegrees=0, swing2LimitDegrees=19, twistLimitDegrees=0}
 {bone1Name=dyn_flap_r_front, bone2Name=base_flap_r_front, swing1LimitDegrees=15, swing2LimitDegrees=25, twistLimitDegrees=0}
 {bone1Name=dyn_flap_r_rear, bone2Name=base_flap_r_rear, swing1LimitDegrees=25, swing2LimitDegrees=25, twistLimitDegrees=0}
 {bone1Name=dyn_chain_c, bone2Name=base_chain_c, swing1LimitDegrees=5, swing2LimitDegrees=60, twistLimitDegrees=0}
 {
  bone1Name=base_flap_r_front
  bone2Name=pelvis
  swing1LimitDegrees=45
  swing1Motion=Free
  swing2LimitDegrees=45
  swing2Motion=Free
  twistLimitDegrees=45
  twistMotion=Free
 }
 {
  bone1Name=base_flap_r_rear
  bone2Name=pelvis
  swing1LimitDegrees=45
  swing1Motion=Free
  swing2LimitDegrees=45
  swing2Motion=Free
  twistLimitDegrees=45
  twistMotion=Free
 }
 {
  bone1Name=base_flap_l_rear
  bone2Name=pelvis
  swing1LimitDegrees=45
  swing1Motion=Free
  swing2LimitDegrees=45
  swing2Motion=Free
  twistLimitDegrees=45
  twistMotion=Free
 }
 {
  bone1Name=base_chain_c
  bone2Name=pelvis
  swing1LimitDegrees=45
  swing1Motion=Free
  swing2LimitDegrees=45
  swing2Motion=Free
  twistLimitDegrees=45
  twistMotion=Free
 }
 {
  bone1Name=base_flap_l_front
  bone2Name=pelvis
  swing1LimitDegrees=45
  swing1Motion=Free
  swing2LimitDegrees=45
  swing2Motion=Free
  twistLimitDegrees=45
  twistMotion=Free
 }
 {
  bone1Name=base_chain_front
  bone2Name=pelvis
  swing1LimitDegrees=45
  swing1Motion=Free
  swing2LimitDegrees=45
  swing2Motion=Free
  twistLimitDegrees=45
  twistMotion=Free
 }
 {
  bone1Name=base_chain_rear
  bone2Name=pelvis
  swing1LimitDegrees=45
  swing1Motion=Free
  swing2LimitDegrees=45
  swing2Motion=Free
  twistLimitDegrees=45
  twistMotion=Free
 }
 {
  bone1Name=thigh_l
  bone2Name=pelvis
  swing1LimitDegrees=45
  swing1Motion=Free
  swing2LimitDegrees=45
  swing2Motion=Free
  twistLimitDegrees=45
  twistMotion=Free
 }
 {
  bone1Name=thigh_r
  bone2Name=pelvis
  swing1LimitDegrees=45
  swing1Motion=Free
  swing2LimitDegrees=45
  swing2Motion=Free
  twistLimitDegrees=45
  twistMotion=Free
 }
]
