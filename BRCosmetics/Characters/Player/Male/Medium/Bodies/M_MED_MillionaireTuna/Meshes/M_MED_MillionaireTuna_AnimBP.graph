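# Animation Blueprint for the MillionaireTuna male body cosmetic, adding secondary motion and corrective posing over the pose copied from its attached parent.
# Control Rig and constraints place chain and pistol bones, then additive bone offsets follow limb rotation alphas.
# AnimDynamics on left arm straps and RigidBody physics run afterwards.
# The event graph updates gravity and those alphas, adjusting for skydive floating, held pistols and front-end display.
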AnimBlueprint /BRCosmetics/Characters/Player/Male/Medium/Bodies/M_MED_MillionaireTuna/Meshes/M_MED_MillionaireTuna_AnimBP
target_skeleton=&/M_MED_MillionaireTuna_Skeleton
graphs [
 AnimGraph {nodes [
   n1 = Anim Dynamics - Bone: dyn_strap_arm_l_top!(
    ComponentPose=Local To Component(LocalPose=Use cached pose 'preSim_cache'())
    Alpha=1
    GravityOverride:Vector="0,0,-980"
    ExternalForce="0,0,0"
   ) -> (Pose) {
    alpha=0.9
    angularDampingOverride=0.98
    angularSpringConstant=1500
    bAngularSpring=true
    bGravityOverrideInSimSpace=true
    bOverrideAngularDamping=true
    bUseGravityOverride=true
    numSolverIterationsPostUpdate=4
    numSolverIterationsPreUpdate=16
    simulationSpace=BoneRelative
    alphaBoolBlend {blendOption=Linear}
    boundBone {boneName=dyn_strap_arm_l_top}
    componentAppliedLinearAccClamp=<60000 60000 60000>
    componentLinearAccScale=<0.3 0.2 0.2>
    componentLinearVelScale=<0.3 0.2 0.2>
    gravityOverride {z=-980}
    physicsBodyDefinitions [{
      sphereCollisionRadius=0
      boundBone {boneName=dyn_strap_arm_l_top}
      boxExtents=<15 5 5>
      constraintSetup {angularTargetAxis=AxisZ, twistAxis=AxisY, angularLimitsMax=<5 12.5 5>, angularLimitsMin=<-5 -12.5 -5>, angularTarget=<0 0 1>}
      localJointOffset=<0 0 0>
     }]
    relativeSpaceBone {boneName=root}
   }
   n2 = Anim Dynamics - Bone: dyn_strap_arm_l_bottom!(ComponentPose=@n1, Alpha=0.5, GravityOverride:Vector="0,0,-980", ExternalForce="0,0,0") -> (Pose) {
    alpha=0.9
    angularDampingOverride=0.98
    angularSpringConstant=1500
    bAngularSpring=true
    bGravityOverrideInSimSpace=true
    bOverrideAngularDamping=true
    bUseGravityOverride=true
    numSolverIterationsPostUpdate=4
    numSolverIterationsPreUpdate=16
    simulationSpace=BoneRelative
    alphaBoolBlend {blendOption=Linear}
    boundBone {boneName=dyn_strap_arm_l_bottom}
    componentAppliedLinearAccClamp=<60000 60000 60000>
    componentLinearAccScale=<0.3 0.2 0.2>
    componentLinearVelScale=<0.3 0.2 0.2>
    gravityOverride {z=-980}
    physicsBodyDefinitions [{
      sphereCollisionRadius=0
      boundBone {boneName=dyn_strap_arm_l_bottom}
      boxExtents=<15 5 5>
      constraintSetup {angularTargetAxis=AxisZ, twistAxis=AxisY, angularLimitsMax=<5 12.5 5>, angularLimitsMin=<-5 -12.5 -5>, angularTarget=<0 0 1>}
      localJointOffset=<0 0 0>
     }]
    relativeSpaceBone {boneName=root}
   }
   n3 = RigidBody!(ComponentPose=@n2, Alpha=Get RB_alpha() -> (RB_alpha:double), OverrideWorldGravity:Vector=Get GravityOverride(), ExternalForce) -> (Pose) {
    bClampLinearTranslationLimitToRefPose=true
    bOverrideWorldGravity=true
    overridePhysicsAsset=&/M_MED_MillionaireTuna_Physics
    simulationSpace=BaseBoneSpace
    alphaBoolBlend {blendOption=Linear}
    baseBoneRef {boneName=pelvis}
    componentAppliedLinearAccClamp=<3000 3000 3000>
    componentLinearAccScale=<0.3 0.3 0.3>
    componentLinearVelScale=<0.3 0.3 0.3>
    simSpaceSettings {maxAngularAcceleration=1000, maxAngularVelocity=3, maxLinearAcceleration=700, maxLinearVelocity=500, velocityScaleZ=0.2, worldAlpha=0.25}
   }
   n4 = "Transform (Modify) Bone - Bone: base_flap_l_rear"(ComponentPose=@n3, Alpha=1, Translation="0,0,0", Rotation="-2.5,0,0", Scale="0.925,1,1") {
    rotationMode=BMM_Additive
    rotationSpace=BCS_BoneSpace
    scaleMode=BMM_Additive
    scaleSpace=BCS_BoneSpace
    translationMode=BMM_Additive
    translationSpace=BCS_BoneSpace
    alphaBoolBlend {blendOption=Linear}
    alphaScaleBiasClamp {interpSpeedDecreasing=30, interpSpeedIncreasing=15}
    boneToModify {boneName=base_flap_l_rear}
    rotation {pitch=-2.5}
    scale {x=0.925}
   }
   Output Pose(Result=Component To Local(ComponentPose=@n4)) {layerGroup=None}
   n5 = Control Rig(Source=Copy Pose From Mesh() {bUseAttachedParent=true}) {
    alphaBoolBlend {blendOption=Linear}
    controlRigAssetReference {blueprintRigClass=&/M_MED_MillionaireTuna_ControlRig.M_MED_MillionaireTuna_ControlRig_C}
   }
   n6 = "Constraint - base_chain_rear (PPPP)"!(
    ComponentPose=Local To Component(LocalPose=@n5)
    Alpha=1
    ConstraintWeights_0=0.5633
    ConstraintWeights_1:float=0.2586
    ConstraintWeights_2:float=0.101
    ConstraintWeights_3:float=0.0771
   ) -> (Pose) {
    alphaBoolBlend {blendOption=Linear}
    boneToModify {boneName=base_chain_rear}
    constraintSetup [{targetBone {boneName=spine_01}}, {targetBone {boneName=dyn_flap_l_rear}}, {targetBone {boneName=pelvis}}, {targetBone {boneName=spine_02}}]
    constraintWeights [1,1,1,1]
   }
   n7 = "Constraint - base_chain_front (PPPP)"!(
    ComponentPose=@n6
    Alpha=1
    ConstraintWeights_0=0.5652
    ConstraintWeights_1:float=0.3574
    ConstraintWeights_2:float=0.0436
    ConstraintWeights_3:float=0.0338
   ) -> (Pose) {
    alphaBoolBlend {blendOption=Linear}
    boneToModify {boneName=base_chain_front}
    constraintSetup [{targetBone {boneName=dyn_flap_l_front}}, {targetBone {boneName=spine_01}}, {targetBone {boneName=spine_02}}, {targetBone {boneName=pelvis}}]
    constraintWeights [1,1,1,1]
   }
   n8 = "Constraint - base_pistol_l (PPPPP)"!(
    ComponentPose=@n7
    Alpha=1
    ConstraintWeights_0=0.6089
    ConstraintWeights_1:float=0.2201
    ConstraintWeights_2:float=0.0892
    ConstraintWeights_3:float=0.0658
    ConstraintWeights_4:float=0.0161
   ) -> (Pose) {
    alphaBoolBlend {blendOption=Linear}
    boneToModify {boneName=base_pistol_l}
    constraintSetup [
     {targetBone {boneName=spine_04}}
     {targetBone {boneName=clavicle_l}}
     {targetBone {boneName=spine_03}}
     {targetBone {boneName=spine_05}}
     {targetBone {boneName=spine_02}}
    ]
    constraintWeights [1,1,1,1,1]
   }
   n9 = "Constraint - base_pistol_r (PPPPP)"!(
    ComponentPose=@n8
    Alpha=1
    ConstraintWeights_0=0.6067
    ConstraintWeights_1:float=0.251
    ConstraintWeights_2:float=0.0835
    ConstraintWeights_3:float=0.048
    ConstraintWeights_4:float=0.0107
   ) -> (Pose) {
    alphaBoolBlend {blendOption=Linear}
    boneToModify {boneName=base_pistol_r}
    constraintSetup [
     {targetBone {boneName=spine_04}}
     {targetBone {boneName=clavicle_r}}
     {targetBone {boneName=spine_03}}
     {targetBone {boneName=spine_05}}
     {targetBone {boneName=spine_02}}
    ]
    constraintWeights [1,1,1,1,1]
   }
   Save cached pose 'start_cache'(Pose=Component To Local(ComponentPose=@n9)) {cacheName=start_cache}
   n10 = "Transform (Modify) Bone - Bone: base_chain_front"(
    ComponentPose=Local To Component(LocalPose=Use cached pose 'start_cache'())
    Alpha=1
    Translation="0,-3,1.75"
    Rotation="0,-25,0"
    Scale="1,1,1"
   ) {
    rotationMode=BMM_Additive
    rotationSpace=BCS_BoneSpace
    translationMode=BMM_Additive
    translationSpace=BCS_BoneSpace
    alphaBoolBlend {blendOption=Linear}
    alphaScaleBiasClamp {interpSpeedDecreasing=30, interpSpeedIncreasing=15}
    boneToModify {boneName=base_chain_front}
    rotation {yaw=-25}
    translation {y=-3, z=1.75}
   }
   n11 = "Transform (Modify) Bone - Bone: base_chain_front"(ComponentPose=@n10, Alpha=Get legRotation_L() -> (legRotation_L:double), Translation="0,-8,-0.5", Rotation="0,0,-45", Scale="1,1,1") {
    rotationMode=BMM_Additive
    rotationSpace=BCS_BoneSpace
    translationMode=BMM_Additive
    translationSpace=BCS_BoneSpace
    alphaBoolBlend {blendOption=Linear}
    alphaScaleBiasClamp {bInterpResult=true, interpSpeedDecreasing=30, interpSpeedIncreasing=15}
    boneToModify {boneName=base_chain_front}
    rotation {roll=-45}
    translation {y=-8, z=-0.5}
   }
   n12 = "Transform (Modify) Bone - Bone: base_chain_c"(ComponentPose=@n11, Alpha=0.5, Translation="0,0,2", Rotation="-15,-45,-35", Scale="1,1,1") {
    rotationMode=BMM_Additive
    rotationSpace=BCS_BoneSpace
    translationMode=BMM_Additive
    translationSpace=BCS_BoneSpace
    alphaBoolBlend {blendOption=Linear}
    alphaScaleBiasClamp {interpSpeedDecreasing=30, interpSpeedIncreasing=15}
    boneToModify {boneName=base_chain_c}
    rotation=<rot -15 -45 -35>
    translation {z=2}
   }
   n13 = "Transform (Modify) Bone - Bone: base_flap_r_rear"(ComponentPose=@n12, Alpha=1, Translation="0,-0.5,1.5", Rotation="0,-10,-7.5", Scale="1,1,1") {
    rotationMode=BMM_Additive
    rotationSpace=BCS_BoneSpace
    translationMode=BMM_Additive
    translationSpace=BCS_BoneSpace
    alphaBoolBlend {blendOption=Linear}
    alphaScaleBiasClamp {interpSpeedDecreasing=30, interpSpeedIncreasing=15}
    boneToModify {boneName=base_flap_r_rear}
    rotation {roll=-7.5, yaw=-10}
    translation {y=-0.5, z=1.5}
   }
   n14 = "Transform (Modify) Bone - Bone: base_flap_l_front"(ComponentPose=@n13, Alpha=Get legRotation_L(), Translation="-1,0,2", Rotation="75,0,30", Scale="1,1,1") {
    rotationMode=BMM_Additive
    rotationSpace=BCS_BoneSpace
    translationMode=BMM_Additive
    translationSpace=BCS_BoneSpace
    alphaBoolBlend {blendOption=Linear}
    boneToModify {boneName=base_flap_l_front}
    rotation {pitch=75, roll=30}
    translation {x=-1, z=2}
   }
   n15 = "Transform (Modify) Bone - Bone: base_flap_r_front"(ComponentPose=@n14, Alpha=Get legRotation_R() -> (legRotation_R:double), Translation="-1.9,-2,0", Rotation="50,0,-10", Scale="1,1,1") {
    rotationMode=BMM_Additive
    rotationSpace=BCS_BoneSpace
    scaleMode=BMM_Additive
    scaleSpace=BCS_BoneSpace
    translationMode=BMM_Additive
    translationSpace=BCS_BoneSpace
    alphaBoolBlend {blendOption=Linear}
    boneToModify {boneName=base_flap_r_front}
    rotation {pitch=50, roll=-10}
    translation=<-1.9 -2>
   }
   n16 = "Transform (Modify) Bone - Bone: base_pistol_l"(ComponentPose=@n15, Alpha=Get UpperArm_RotNeg_L() -> (UpperArm_RotNeg_L:double), Translation="7,0,0", Rotation="-15,-100,0", Scale="1,1,1") {
    rotationMode=BMM_Additive
    rotationSpace=BCS_BoneSpace
    translationMode=BMM_Additive
    translationSpace=BCS_BoneSpace
    alphaBoolBlend {blendOption=Linear}
    alphaScaleBiasClamp {bInterpResult=true}
    boneToModify {boneName=base_pistol_l}
    rotation {pitch=-15, yaw=-100}
    translation {x=7}
   }
   n17 = "Transform (Modify) Bone - Bone: base_pistol_r"(ComponentPose=@n16, Alpha=Get UpperArm_RotNeg_R() -> (UpperArm_RotNeg_R:double), Translation="7,0,0", Rotation="-15,100,0", Scale="1,1,1") {
    rotationMode=BMM_Additive
    rotationSpace=BCS_BoneSpace
    translationMode=BMM_Additive
    translationSpace=BCS_BoneSpace
    alphaBoolBlend {blendOption=Linear}
    boneToModify {boneName=base_pistol_r}
    rotation {pitch=-15, yaw=100}
    translation {x=7}
   }
   n18 = "Transform (Modify) Bone - Bone: deform_upperarm_up_base_r"(
    ComponentPose=@n17
    Alpha=Get UpperArm_RotPos_R() -> (UpperArm_RotPos_R:double)
    Translation=Get translationValve_R() -> (translationValve_R:Vector)
    Rotation=Get rotationValve_R() -> (rotationValve_R:Rotator)
    Scale="1,1,1"
   ) {
    rotationMode=BMM_Additive
    rotationSpace=BCS_BoneSpace
    translationMode=BMM_Additive
    translationSpace=BCS_BoneSpace
    alphaBoolBlend {blendOption=Linear}
    boneToModify {boneName=deform_upperarm_up_base_r}
    rotation {yaw=60}
    translation {x=-10, z=-1}
   }
   n19 = Get UpperArm_RotPos_L() -> (UpperArm_RotPos_L:double)
   n20 = "Transform (Modify) Bone - Bone: deform_upperarm_up_base_l"(
    ComponentPose=@n18
    Alpha=@n19
    Translation=Get translationValve_L() -> (translationValve_L:Vector)
    Rotation=Get rotationValve_L() -> (rotationValve_L:Rotator)
    Scale="1,1,1"
   ) {
    rotationMode=BMM_Additive
    rotationSpace=BCS_BoneSpace
    translationMode=BMM_Additive
    translationSpace=BCS_BoneSpace
    alphaBoolBlend {blendOption=Linear}
    boneToModify {boneName=deform_upperarm_up_base_l}
    rotation {yaw=60}
    translation {x=10, z=1}
   }
   n21 = "Transform (Modify) Bone - Bone: base_strap_arm_l_top"(ComponentPose=@n20, Alpha=@n19, Translation="0,-2,0", Rotation="0,0,25", Scale="1,1,1") {
    rotationMode=BMM_Additive
    rotationSpace=BCS_BoneSpace
    translationMode=BMM_Additive
    translationSpace=BCS_BoneSpace
    alphaBoolBlend {blendOption=Linear}
    boneToModify {boneName=base_strap_arm_l_top}
    rotation {roll=25}
    translation {y=-2}
   }
   n22 = "Transform (Modify) Bone - Bone: base_knee_pad_r"(ComponentPose=@n21, Alpha=Get Knee_Pad_Rot() -> (Knee_Pad_Rot:double), Translation="1,0,-1", Rotation="80,0,0", Scale="1,1,1") {
    rotationMode=BMM_Additive
    rotationSpace=BCS_BoneSpace
    translationMode=BMM_Additive
    translationSpace=BCS_BoneSpace
    alphaBoolBlend {blendOption=Linear}
    boneToModify {boneName=base_knee_pad_r}
    rotation {pitch=80}
    translation {x=1, z=-1}
   }
   Save cached pose 'preSim_cache'(Pose=Component To Local(ComponentPose=@n22)) {cacheName=preSim_cache}
  ]}
 EventGraph {nodes [
   Event BlueprintUpdateAnimation()
   > n1 = Sequence!(execute) -> (then_0, then_1, then_2:Exec, then_3:Exec)
   > n2 = Gravity Override State Function DEPRECATED(
    CharacterPartAnimInstance=Self-Reference()
    GameTimeInSeconds=GetGameTimeInSeconds()
    GravityOverrideParams=Get GravityOverrideParams() -> (GravityOverrideParams:Gravity Override Params Struct Structure)
   )
   > Set GravityOverride(GravityOverride=@n2)
   > Set SkelMesh(SkelMesh=GetPartSkeletalMeshComponent())
   > Set Knee_Pad_Rot(Knee_Pad_Rot:double=GetAlphaFromReferenceTransformDelta(SkeletalMeshComponent=Get SkelMesh(), BoneName=calf_r, BaseName=thigh_r, InputRange=(X=0,Y=120), TwistAxis="0,0,1")) -> (Output_Get:double)
   n3 = Get SkelMesh()
   > Set legRotation_L(legRotation_L:double=GetAlphaFromReferenceTransformDelta(SkeletalMeshComponent=@n3, BoneName=pelvis, BaseName=thigh_l, InputRange=(X=0,Y=90), TwistAxis="0,0,1")) -> (Output_Get:double)
   > Set legRotation_R(legRotation_R:double=GetAlphaFromReferenceTransformDelta(SkeletalMeshComponent=@n3, BoneName=pelvis, BaseName=thigh_r, InputRange=(X=0,Y=90), TwistAxis="0,0,1")) -> (Output_Get:double)
   > n4 = Branch(Condition=Get bIsSkydiveFloating())
   > Set translationValve_L(translationValve_L:Vector="10,0,1") -> (Output_Get:Vector)
   > Set rotationValve_L(rotationValve_L:Rotator="0,60,0") -> (Output_Get:Rotator)
   > Set translationValve_R(translationValve_R:Vector="-10,0,-1") -> (Output_Get:Vector)
   > Set rotationValve_R(rotationValve_R:Rotator="0,60,0") -> (Output_Get:Rotator)
   Set translationValve_L(execute=@n4.else, translationValve_L="8,0,0")
   > Set rotationValve_L(rotationValve_L="-50,25,-50")
   > Set translationValve_R(translationValve_R="-8,0,-1")
   > Set rotationValve_R(rotationValve_R="-50,-25,-50")
   n5 = Branch(execute=@n1.then_3, Condition=Get bIsInFrontEnd())
   > Set RB_alpha(RB_alpha:double=0.85) -> (Output_Get:double)
   Set RB_alpha(execute=@n5.else, RB_alpha=1)
   n6 = Get SkelMesh()
   Set UpperArm_RotNeg_R(
    execute=@n1.then_1
    UpperArm_RotNeg_R:double=GetAlphaFromReferenceTransformDelta(SkeletalMeshComponent=@n6, BoneName=upperarm_r, BaseName=clavicle_r, DeltaComponentType=EulerY, InputRange=(X=-15,Y=-110), TwistAxis="0,0,1")
   ) -> (Output_Get:double)
   > Set UpperArm_RotNeg_L(UpperArm_RotNeg_L:double=GetAlphaFromReferenceTransformDelta(SkeletalMeshComponent=@n6, BoneName=upperarm_l, BaseName=clavicle_l, DeltaComponentType=EulerY, InputRange=(X=-15,Y=-110), TwistAxis="0,0,1")) -> (Output_Get:double)
   n7 = NOT Boolean(A=Get bIsSkyDiving())
   > n8 = Branch(Condition=AND Boolean(A=Get isUsingDoublePistols() -> (isUsingDoublePistols:bool), B=@n7))
   > Set UpperArm_RotPos_L(UpperArm_RotPos_L:double=0) -> (Output_Get:double)
   n9 = Get SkelMesh()
   n10 = GetAlphaFromReferenceTransformDelta(SkeletalMeshComponent=@n9, BoneName=upperarm_r, BaseName=clavicle_r, DeltaComponentType=EulerY, InputRange=(X=0,Y=25), TwistAxis="0,0,1")
   n11 = "Clamp (Float)"(Value=@n10, Min=0, Max=0.5)
   > Set UpperArm_RotPos_R(UpperArm_RotPos_R:double=@n11) -> (Output_Get:double)
   n12 = Branch(execute=@n8.else, Condition=AND Boolean(A=@n7, B=Get isUsingPistol() -> (isUsingPistol:bool)))
   n13 = GetAlphaFromReferenceTransformDelta(SkeletalMeshComponent=@n9, BoneName=upperarm_l, BaseName=clavicle_l, DeltaComponentType=EulerY, InputRange=(X=0,Y=25), TwistAxis="0,0,1")
   > Set UpperArm_RotPos_L(UpperArm_RotPos_L="Clamp (Float)"(Value=@n13, Min=0, Max=0.75))
   > Set UpperArm_RotPos_R(UpperArm_RotPos_R=@n11)
   Set UpperArm_RotPos_L(execute=@n12.else, UpperArm_RotPos_L=@n13)
   > Set UpperArm_RotPos_R(UpperArm_RotPos_R=@n10)
   n14 = Cast To FortPlayerAnimInstance(execute=@n1.then_2, Object=GetAnimInstance(self=GetOwnerSkeletalMeshComponent()))
   > Set isUsingPistol(isUsingPistol:bool=false) -> (Output_Get:bool)
   > Set isUsingDoublePistols(isUsingDoublePistols:bool=false) -> (Output_Get:bool)
   > n15 = Switch on EFortWeaponCoreAnimation!(execute, Selection:EFortWeaponCoreAnimation Enum=Get WeaponCoreAnim(self=@n14)) -> (
    Melee:Exec
    Pistol:Exec
    Shotgun:Exec
    PaperBlueprint:Exec
    Rifle:Exec
    MeleeOneHand:Exec
    MachinePistol:Exec
    RocketLauncher:Exec
    GrenadeLauncher:Exec
    GoingCommando:Exec
    AssaultRifle:Exec
    TacticalShotgun:Exec
    SniperRifle:Exec
    TrapPlacement:Exec
    ShoulderLauncher:Exec
    AbilityDecoTool:Exec
    Crossbow:Exec
    C4:Exec
    RemoteControl:Exec
    DualWield:Exec
    AR_BullPup:Exec
    AR_ForwardGrip:Exec
    MedPackPaddles:Exec
    SMG_P90:Exec
    AR_DrumGun:Exec
    Consumable_Small:Exec
    Consumable_Large:Exec
    Balloon:Exec
    MountedTurret:Exec
    CreativeTool:Exec
    ExplosiveBow:Exec
    AshtonIndigo:Exec
    AshtonChicago:Exec
    MeleeDualWield:Exec
    Unarmed:Exec
   )
   Set isUsingDoublePistols(execute=@n15.DualWield, isUsingDoublePistols=true)
   Set isUsingPistol(execute=@n15.Pistol, isUsingPistol=true)
   TryGetPawnOwner()
  ]}
]
variables [
 GravityOverride {type=Vector}
 GravityOverrideParams {type=GravityOverrideParamsStruct}
 isUsingDoublePistols {type=boolean}
 isUsingPistol {type=boolean}
 Knee_Pad_Rot {type=number}
 legRotation_L {type=number}
 legRotation_R {type=number}
 RB_alpha {type=number}
 rotationValve_L {type=Rotator}
 rotationValve_R {type=Rotator}
 SkelMesh {type=SkeletalMeshComponent}
 translationValve_L {type=Vector}
 translationValve_R {type=Vector}
 UpperArm_RotNeg_L {type=number}
 UpperArm_RotNeg_R {type=number}
 UpperArm_RotPos_L {type=number}
 UpperArm_RotPos_R {type=number}
]
